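# Niagara system for yacht engine smoke, built from two continuously spawning GPU sprite emitters.
# JetStream emits short-lived, tall, faster puffs, while Smoke1 emits longer-lived, square, randomly rotated puffs.
# Both launch grey, translucent particles from the system location in an upward, slightly Y-tilted cone, slowed by drag.
# Curves drive alpha fade, sprite growth and flipbook frames; engine-mesh sampling, burst and scalability modules are disabled.
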NiagaraSystem /Game/Effects/Fort_Effects/BluePrints/Environments/Yacht/NS_Yacht_Engine_Smoke
dependencies {
 usedDataInterfaces [&NiagaraDataInterfaceCurve, &NiagaraDataInterfaceStaticMesh]
 usedDynamicInputs []
 usedModules [&SystemState]
}
emitters {
 JetStream {
  simTarget=GPUComputeSim
  emitterUpdateScript [
   &EmitterScalability(
    Module.bApplySpawnCountScale=false
    Module.bApplySpawnCountScaleByDistanceFraction=false
    Module.bUseMaxDistance=false
    Module.bUseMinDistance=false
   ) {enabled=false}
   &EmitterLifeCycle() {enabled=false}
   &EmitterState()
   &SpawnBurst_Instantaneous(Spawn Count=5, Use Spawn Probability=false) {enabled=false}
   &SpawnRate(SpawnRate=6, Use Spawn Probability=false)
  ]
  particleSpawnScript [
   /Niagara/Modules/Spawn/Initialization/InitializeParticle(
    Color=<rgba 0.5677 0.5677 0.5677 0.5>
    Lifetime=0.4
    Module.Write Color=true
    Module.Write Lifetime=true
    Module.Write Mass=true
    Module.Write Material Random=true
    Module.Write Position=false
    Module.Write Scale=false
    Module.Write SpriteRotation=false
    Module.Write SpriteSize=true
    Position=&SimulationPosition()
    Sprite Size=<50 150>
   )
   &SystemLocation(MaskSpawnGroup=false)
   &AddVelocityInCone(Cone Angle=20, Cone Axis=<0 0.2 1>, Module.Use Velocity Falloff On Cone Axis=false, Velocity Strength=200)
   &ApplyInitialForces()
   /Niagara/Modules/Spawn/MeshInterface/SampleStaticMesh(
    OverrideRandomSeed=false
    StaticMesh=DataInterface() {propertyValues {
      DefaultMesh=&/SM_Yacht_Engine_Emitter
      InstanceIndex=-1
      LODIndex=0
      PreviewMesh=null
      SoftSourceActor=null
      SourceMode=Default
      bAllowSamplingFromStreamingLODs=false
      bCaptureTransformsPerFrame=true
      bUsePhysicsBodyVelocity=false
      FilteredSockets []
      LODIndexUserParameter {parameter {name=None, typeDefHandle {registeredTypeIndex=108}, varData []}}
      MeshParameterBinding {parameter {name=None, typeDefHandle {registeredTypeIndex=1}, varData []}}
      SectionFilter {allowedMaterialSlots []}
     }}
   ) {enabled=false}
   /Niagara/Modules/Spawn/Location/StaticMeshLocation(Module.Enable Axis Constraint=false) {enabled=false}
  ]
  particleUpdateScript [
   &ParticleState()
   &Drag(Use Linear Drag=true, Use Rotational Drag=false)
   &SolveForcesAndVelocity(Clamp Velocity=false, Limit Acceleration=false)
   &ScaleColor(
    Scale Alpha=&FloatFromCurve(FloatCurve=DataInterface() {propertyValues {Curve {keys [<at 0 0>,<at 0.0917 0.9657>,<at 1.0011 0.002>]}}})
    ScaleA=true
    ScaleRGB=false
   )
   &ScaleSpriteSize(Scale Factor=&Vector2DFromFloat(Value=&FloatFromCurve(FloatCurve=DataInterface() {propertyValues {Curve {keys [<at -0.1527 0.1179>,<at 0.9919 0.7352>]}}})))
   /SetVariables(Particles.SubImageIndex=&FloatFromCurve(FloatCurve=DataInterface() {propertyValues {Curve {keys [<at 0 0>,<at 0.1543 0.2735>,<at 0.4251 0.6158>,<at 1.0011 1>]}}}, Scale Curve=32)) {isSetParameters=true, module=SetVariables_#1}
  ]
  rendererClasses [&NiagaraSpriteRendererProperties]
  renderers [{rendererClass=&NiagaraSpriteRendererProperties}]
 }
 Smoke1 {
  simTarget=GPUComputeSim
  emitterUpdateScript [
   &EmitterScalability(
    Module.bApplySpawnCountScale=false
    Module.bApplySpawnCountScaleByDistanceFraction=false
    Module.bUseMaxDistance=false
    Module.bUseMinDistance=false
   ) {enabled=false}
   &EmitterLifeCycle() {enabled=false}
   &EmitterState()
   &SpawnBurst_Instantaneous(Spawn Count=5, Use Spawn Probability=false) {enabled=false}
   &SpawnRate(SpawnRate=6, Use Spawn Probability=false)
  ]
  particleSpawnScript [
   &InitializeParticle(
    Color=<rgba 0.5677 0.5677 0.5677 0.5>
    Lifetime=&UniformRangedFloat(Minimum=0.7, Maximum=0.8)
    Module.Write Color=true
    Module.Write Lifetime=true
    Module.Write Mass=true
    Module.Write Material Random=true
    Module.Write Position=false
    Module.Write Scale=false
    Module.Write SpriteRotation=true
    Module.Write SpriteSize=true
    Position=&SimulationPosition()
    Sprite Rotation=&UniformRangedFloat(Minimum=0, Maximum=360)
    Sprite Size=&Vector2DFromFloat(Value=&UniformRangedFloat(Minimum=100, Maximum=100))
   )
   &SystemLocation(MaskSpawnGroup=false)
   &AddVelocityInCone(Cone Angle=20, Cone Axis=<0 0.2 1>, Module.Use Velocity Falloff On Cone Axis=false, Velocity Strength=150)
   &ApplyInitialForces()
   &SampleStaticMesh(
    OverrideRandomSeed=false
    StaticMesh=DataInterface() {propertyValues {
      DefaultMesh=&SM_Yacht_Engine_Emitter
      InstanceIndex=-1
      LODIndex=0
      PreviewMesh=null
      SoftSourceActor=null
      SourceMode=Default
      bAllowSamplingFromStreamingLODs=false
      bCaptureTransformsPerFrame=true
      bUsePhysicsBodyVelocity=false
      FilteredSockets []
      LODIndexUserParameter {parameter {name=None, typeDefHandle {registeredTypeIndex=108}, varData []}}
      MeshParameterBinding {parameter {name=None, typeDefHandle {registeredTypeIndex=1}, varData []}}
      SectionFilter {allowedMaterialSlots []}
     }}
   ) {enabled=false}
   &StaticMeshLocation(Module.Enable Axis Constraint=false) {enabled=false}
  ]
  particleUpdateScript [
   &ParticleState()
   &Drag(Use Linear Drag=true, Use Rotational Drag=false)
   &SolveForcesAndVelocity(Clamp Velocity=false, Limit Acceleration=false)
   &ScaleColor(
    Scale Alpha=&FloatFromCurve(FloatCurve=DataInterface() {propertyValues {Curve {keys [<at 0 0>,<at 0.1974 1>,<at 1.0011 0.002>]}}})
    ScaleA=true
    ScaleRGB=false
   )
   &ScaleSpriteSize(Scale Factor=&Vector2DFromFloat(Value=&FloatFromCurve(FloatCurve=DataInterface() {propertyValues {Curve {keys [<at 0.0024 0.3753>,<at 0.4766 0.6713>,<at 0.9891 1>]}}})))
   /SetVariables(Particles.SubImageIndex=&FloatFromCurve(FloatCurve=DataInterface() {propertyValues {Curve {keys [<at 0 0>,<at 0.1543 0.2735>,<at 0.4251 0.6158>,<at 1.0011 1>]}}}, Scale Curve=32)) {isSetParameters=true, module=SetVariables_#1}
  ]
  rendererClasses [&NiagaraSpriteRendererProperties]
  renderers [{rendererClass=&NiagaraSpriteRendererProperties}]
 }
}
summary {
 systemName=NS_Yacht_Engine_Smoke
 emitters [
  {emitterName=Smoke1, simTarget=GPUComputeSim, rendererClasses [&NiagaraSpriteRendererProperties]}
  {emitterName=JetStream, simTarget=GPUComputeSim, rendererClasses [&NiagaraSpriteRendererProperties]}
 ]
 userVariables []
}
